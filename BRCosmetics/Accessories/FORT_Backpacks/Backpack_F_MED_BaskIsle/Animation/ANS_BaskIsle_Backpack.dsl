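; Anim notify state for the BaskIsle back bling that feeds a book bone's position into the backpack's materials during animation.
; On each notify tick it reads the component-space transform of root_book_jnt on the animated skeletal mesh.
; It writes that location to the BookPosition vector parameter on all of the mesh's materials, so shaders can track the moving book.
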(fn Received_NotifyTick
  (params (MeshComp :type Skeletal%20Mesh%20Component%20Object%20Reference) (Animation :type Anim%20Sequence%20Base%20Object%20Reference) (FrameDeltaTime :type Float%20%28single-precision%29) (EventReference :type Anim%20Notify%20Event%20Reference%20Structure%20%28by%20ref%29))
  (bind _returnvalue_location (Math|Transform|BreakTransform (Transformation|GetBoneTransform MeshComp "root_book_jnt" "RTS_Component")))
  (Rendering|Material|SetVectorParameterValueonMaterials MeshComp "BookPosition" _returnvalue_location)
  (return false))
