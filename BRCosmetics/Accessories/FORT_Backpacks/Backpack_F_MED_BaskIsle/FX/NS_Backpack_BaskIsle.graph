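# Niagara effect for the BaskIsle backpack cosmetic, showing socket-attached fire meshes, glowing eyes, cyan-blue rising sparks and flipbook fire sprites.
# A shared System Timeline drives fire mesh scale and color, spark spawning, and a short flipbook burst near its end.
# FireHead stores a smoothed world position for lagged motion data.
# Time-of-day scalars adjust alpha on every emitter except Sparks, which uses curl noise and upward force.
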NiagaraSystem /BRCosmetics/Accessories/FORT_Backpacks/Backpack_F_MED_BaskIsle/FX/NS_Backpack_BaskIsle
dependencies {
 usedDataInterfaces [
  &NiagaraDataInterfaceCurve
  &NiagaraDataInterfaceArrayFloat
  &NiagaraDataInterfaceMeshRendererInfo
  &NiagaraDataInterfaceVector2DCurve
  &NiagaraDataInterfaceSpriteRendererInfo
  &NiagaraDataInterfaceSocketReader
 ]
 usedDynamicInputs [
  &SelectFloatFromArray
  &Vector2DFromFloat
  &Vector2DFromCurve
  &RandomRangeVector2D
  &Multiply_FloatByInt
  &ReturnParticleIDAttributes
  &SetBoolByIntComparison
 ]
 usedModules [
  &self:SystemSpawnScript.SystemScriptSource.SystemScriptGraph.NiagaraNodeAssignment_2.SetVariables_#1
  &self:SystemSpawnScript.SystemScriptSource.SystemScriptGraph.NiagaraNodeAssignment_4.SetVariables_#2
  &SystemState
  &self:SystemSpawnScript.SystemScriptSource.SystemScriptGraph.NiagaraNodeAssignment_5.SetVariables_#3
  &Timeline
  &self:SystemSpawnScript.SystemScriptSource.SystemScriptGraph.NiagaraNodeAssignment_6.SetVariables_#4
 ]
}
emitters {
 EyeGlow {
  emitterUpdateScript [
   &EmitterState(
    Inactive Response=::NewEnumerator1
    Life Cycle Mode=::NewEnumerator1
    Loop Behavior=::NewEnumerator1
    Loop Duration Mode=::NewEnumerator1
    UseLoopDelay=false
   )
   &SpawnBurst_Instantaneous(Spawn Count=2, Use Spawn Probability=false, UseLoopCountLimit=false)
  ]
  particleSpawnScript [
   &FN_SetSpawnGroup(Number of Spawn Groups=2)
   &InitializeParticle(
    Color=<rgba 0.0833 0.7498 1 1>
    Mesh Scale=<1 0.8 0.8>
    Mesh Scale Mode=::NewEnumerator4
    Sprite Size Mode=::NewEnumerator3
    Uniform Sprite Size=5
    UsePositionOffset=false
    Write Material Random=true
   )
   &CameraOffset(Camera Offset Amount=2)
  ]
  particleUpdateScript [
   &ParticleState(Kill Particles When Lifetime Has Elapsed=false, Let Infinitely Lived Particles Die When Emitter Deactivates=true)
   &ScaleSpriteSize(Scale Sprite Size Mode=::NewEnumerator0)
   &ScaleColor(
    Scale Alpha=&Multiply_Float(
     A=0.4
     B=&Multiply_Float(A=&FNTimeOfDayScalar(Night Scale=0.6, Morning Scale=0.9, Evening Scale=0.8), B=1)
    )
    ScaleA=true
    ScaleRGB=true
   )
   &SocketLocation(
    SamplePosOffset=true
    Socket Index=&MergeHelper_IntSwitch(0=3, 1=4, 2=?, 3=?, 4=?, 5=?, Index=&SpawnGroup(), Num=::NewEnumerator0)
    Socket Reader=$System.Socket Reader
    Socket Sampling Mode=::NewEnumerator2
    VelocitySpeedLimit=false
   )
  ]
  rendererClasses [&NiagaraSpriteRendererProperties]
  renderers [{rendererClass=&NiagaraSpriteRendererProperties}]
 }
 FireHead {
  emitterUpdateScript [
   &EmitterState(
    Inactive Response=::NewEnumerator1
    Life Cycle Mode=::NewEnumerator1
    Loop Behavior=::NewEnumerator1
    Loop Duration Mode=::NewEnumerator1
    UseLoopDelay=false
   )
   &SpawnBurst_Instantaneous(Spawn Count=3, Use Spawn Probability=false, UseLoopCountLimit=false)
  ]
  particleSpawnScript [
   &FN_SetSpawnGroup(Number of Spawn Groups=3)
   &InitializeParticle(
    Mesh Scale=&Multiply_VectorByFloat(
     Vector=<0.9 1 1.3>
     Float=&MergeHelper_FloatSwitch(0=0.3, 1=0.1, 2=0.1, 3=?, Index=&SpawnGroup(), Num=::NewEnumerator1)
    )
    Mesh Scale Mode=::NewEnumerator4
    Position Offset=<-10 0 0>
    UsePositionOffset=true
    Write Material Random=true
   )
   &InitialMeshOrientation(Mesh Axis To Orient=<0 1 0>, Mesh Orientation Mode=::NewEnumerator2, Rotation=<0 -0.25 0>, Use Rotation Vector=true)
   &CameraOffset(Camera Offset Amount=&MergeHelper_FloatSwitch(0=-4, 1=1, 2=1, 3=?, Index=&SpawnGroup(), Num=::NewEnumerator1))
   &DynamicMaterialParameters(
    Index 0 Param 1=1.2
    Module.Index0Float0Write=true
    Module.Index0Float1Write=true
    Module.Index0Float2Write=true
    Module.Index0Float3Write=true
   )
  ]
  particleUpdateScript [
   &ParticleState(Kill Particles When Lifetime Has Elapsed=false, Let Infinitely Lived Particles Die When Emitter Deactivates=true)
   &ScaleMeshSize(Scale Factor=&VectorFromFloat(Value=&FloatFromCurve(
      FloatCurve=DataInterface() {propertyValues {Curve {keys [
          <at 0 0 RCIM_Cubic>
          {arriveTangent=-0.1618, interpMode=RCIM_Cubic, leaveTangent=-0.1618, tangentMode=RCTM_User, time=0.0144, value=0.9803}
          {arriveTangent=0.0122, interpMode=RCIM_Cubic, leaveTangent=0.0122, tangentMode=RCTM_User, time=0.983, value=1}
          <at 1 0>
         ]}}}
      CurveIndex=$System.Timeline.Time
     )))
   &InterpolateOverTime(
    Rate Of Change=15
    Target Vector Value:Vector3f=&TransformPosition(Position=$Particles.Position, Source Space=::Local, Destination Space=::World)
    Variable Type=::NewEnumerator2
   )
   /SetVariables(Particles.InterpolateDY:Vector4f=&MakeFloat4FromVectorAndFloat("Vector (RGB)"=$StackContext.InterpolateOverTime.InterpolateOverTime_V3_.Moving Average, "Float (Alpha)"=1)) {isSetParameters=true}
   &SocketLocation(
    SamplePosOffset=false
    Socket Index=&MergeHelper_IntSwitch(0=2, 1=5, 2=6, 3=?, 4=?, 5=?, Index=&SpawnGroup(), Num=::NewEnumerator1)
    Socket Reader=$System.Socket Reader
    Socket Sampling Mode=::NewEnumerator3
    VelocitySpeedLimit=false
   )
   &ScaleColor(
    Scale Alpha=&FloatFromCurve(
     FloatCurve=DataInterface() {propertyValues {Curve {keys [
         <at 0.1 0 RCIM_Cubic>
         {arriveTangent=-0.2061, interpMode=RCIM_Cubic, leaveTangent=-0.2061, tangentMode=RCTM_User, time=0.1486, value=1.5774}
         {arriveTangent=-0.1618, interpMode=RCIM_Cubic, leaveTangent=-0.1618, tangentMode=RCTM_User, time=0.2843, value=0.9803}
         {arriveTangent=0.0122, interpMode=RCIM_Cubic, leaveTangent=0.0122, tangentMode=RCTM_User, time=0.95, value=1}
         <at 1 0>
        ]}}}
     CurveIndex=$System.Timeline.Time
     Scale Curve=&Multiply_Float(
      A=&MergeHelper_FloatSwitch(0=1, 1=2, 2=2, 3=?, Index=&SpawnGroup(), Num=::NewEnumerator1)
      B=&FNTimeOfDayScalar(Night Scale=0.6, Morning Scale=0.9, Evening Scale=0.8)
     )
    )
    Scale RGB=&VectorFromFloat(Value=&FloatFromCurve(
      FloatCurve=DataInterface() {propertyValues {Curve {keys [
          <at 0.1 1 RCIM_Cubic>
          {arriveTangent=-0.2061, interpMode=RCIM_Cubic, leaveTangent=-0.2061, tangentMode=RCTM_User, time=0.1486, value=1.5774}
          {arriveTangent=-0.1618, interpMode=RCIM_Cubic, leaveTangent=-0.1618, tangentMode=RCTM_User, time=0.2843, value=0.9803}
          {arriveTangent=0.0122, interpMode=RCIM_Cubic, leaveTangent=0.0122, tangentMode=RCTM_User, time=0.9299, value=1}
          <at 1 1>
         ]}}}
      CurveIndex=$System.Timeline.Time
     ))
    ScaleA=true
    ScaleRGB=true
   )
  ]
  rendererClasses [&NiagaraMeshRendererProperties]
  renderers [{rendererClass=&NiagaraMeshRendererProperties, rendererIndex=0}]
 }
 FireSpline {
  emitterUpdateScript [
   &EmitterState(
    Inactive Response=::NewEnumerator1
    Life Cycle Mode=::NewEnumerator1
    Loop Behavior=::NewEnumerator1
    Loop Duration Mode=::NewEnumerator1
    UseLoopDelay=false
   )
   &SpawnBurst_Instantaneous(Spawn Count=2, Use Spawn Probability=false, UseLoopCountLimit=false)
  ]
  particleSpawnScript [
   &FN_SetSpawnGroup()
   &InitializeParticle(
    Lifetime=3
    Mesh Renderer Array Visibility Mode=::NewEnumerator1
    Mesh Renderer Info=DataInterface() {propertyValues {MeshRenderer=&self:FireSpline_0.NiagaraMeshRendererProperties_0}}
    Mesh Renderer Mesh Index:NiagaraInt32=&SpawnGroup()
    Mesh Scale Mode=::NewEnumerator4
    UsePositionOffset=false
    Write Material Random=true
   )
   &CameraOffset(Camera Offset Amount=5)
  ]
  particleUpdateScript [
   &ParticleState(
    Kill Particles When Lifetime Has Elapsed=false
    Let Infinitely Lived Particles Die When Emitter Deactivates=true
    Loop Particles Lifetime=true
   )
   &ScaleMeshSize(Scale Factor=&VectorFromFloat(Value=&FloatFromCurve(
      FloatCurve=DataInterface() {propertyValues {Curve {keys [
          <at 0 0 RCIM_Cubic>
          {arriveTangent=-0.1618, interpMode=RCIM_Cubic, leaveTangent=-0.1618, tangentMode=RCTM_User, time=0.0144, value=0.9803}
          {arriveTangent=0.0122, interpMode=RCIM_Cubic, leaveTangent=0.0122, tangentMode=RCTM_User, time=0.983, value=1}
          <at 1 0>
         ]}}}
      CurveIndex=$System.Timeline.Time
      Scale Curve=1.14
     )))
   &ScaleColor(
    Scale Alpha=&FloatFromCurve(
     FloatCurve=DataInterface() {propertyValues {Curve {keys [
         <at 0.1 0 RCIM_Cubic>
         {arriveTangent=-0.2061, interpMode=RCIM_Cubic, leaveTangent=-0.2061, tangentMode=RCTM_User, time=0.1486, value=1.5774}
         {arriveTangent=-0.1618, interpMode=RCIM_Cubic, leaveTangent=-0.1618, tangentMode=RCTM_User, time=0.2843, value=0.9803}
         {arriveTangent=0.0122, interpMode=RCIM_Cubic, leaveTangent=0.0122, tangentMode=RCTM_User, time=0.95, value=1}
         <at 1 0>
        ]}}}
     CurveIndex=$System.Timeline.Time
     Scale Curve=&Multiply_Float(A=1.2, B=&FNTimeOfDayScalar(Night Scale=0.6, Morning Scale=0.9, Evening Scale=0.8))
    )
    Scale RGB=&VectorFromFloat(Value=&FloatFromCurve(
      FloatCurve=DataInterface() {propertyValues {Curve {keys [
          {arriveTangent=-2.5331, interpMode=RCIM_Cubic, time=0.1, value=1}
          {arriveTangent=-0.2061, interpMode=RCIM_Cubic, leaveTangent=-0.2061, tangentMode=RCTM_User, time=0.1486, value=1.5774}
          {arriveTangent=-0.1618, interpMode=RCIM_Cubic, leaveTangent=-0.1618, tangentMode=RCTM_User, time=0.2843, value=0.9803}
          {arriveTangent=0.0122, interpMode=RCIM_Cubic, leaveTangent=0.0122, tangentMode=RCTM_User, time=0.9299, value=1}
          <at 1 1>
         ]}}}
      CurveIndex=$System.Timeline.Time
     ))
    ScaleA=true
    ScaleRGB=true
   )
   &SocketLocation(
    SamplePosOffset=false
    Socket Index=&SpawnGroup()
    Socket Reader=$System.Socket Reader
    Socket Sampling Mode=::NewEnumerator3
    VelocitySpeedLimit=false
   )
  ]
  rendererClasses [&NiagaraMeshRendererProperties]
  renderers [
   {rendererClass=&NiagaraMeshRendererProperties, rendererIndex=0}
   {rendererClass=&NiagaraMeshRendererProperties, rendererIndex=1}
  ]
 }
 FiresprintSubUV {
  emitterUpdateScript [
   &EmitterState(Life Cycle Mode=::NewEnumerator1, Loop Duration=1, UseLoopDelay=false)
   &SpawnRate(
    SpawnRate=&Multiply_Float(
     A=60
     B=&FloatFromCurve(
      FloatCurve=DataInterface() {propertyValues {Curve {keys [<at 0.9121 0>,<at 0.9209 1>,<at 0.98 1>,<at 0.99 0>]}}}
      CurveIndex=$System.Timeline.Time
     )
    )
    Use Spawn Probability=false
   )
  ]
  particleSpawnScript [
   &InitializeParticle(
    Color=<rgba 0 2.2041 5 2>
    Lifetime Max=0.4
    Lifetime Min=0.3
    Lifetime Mode=::NewEnumerator1
    Sprite Rotation Mode=::NewEnumerator2
    Sprite Size Mode=::NewEnumerator3
    Uniform Sprite Size=&Multiply_Float(A=10, B=&RandomRangeFloat(Minimum=0, Maximum=1))
    UsePositionOffset=true
    Write Material Random=true
   )
   &CameraOffset(Camera Offset Amount=3)
   &ShapeLocation(
    MaskSpawnGroup=false
    Offset=<0 -11 -3>
    Offset Mode=::NewEnumerator0
    Override Seed=false
    Rotation Mode=::NewEnumerator1
    Sphere Radius=8
    UseOwnerScale=true
   )
  ]
  particleUpdateScript [
   &ParticleState()
   &ScaleSpriteSize(
    Scale Sprite Size Mode=::NewEnumerator0
    Uniform Scale Factor=&FloatFromCurve(FloatCurve=DataInterface() {propertyValues {Curve {keys [<at 0 1.5>,<at 1 1>]}}}, Scale Curve=2.3)
   )
   &SpriteRotationRate(Rotation Rate=&RandomRangeFloat(Minimum=-50, Maximum=50))
   &ScaleColor(
    Scale Alpha=&Multiply_Float(
     A=&FNTimeOfDayScalar(Night Scale=0.7, Morning Scale=0.9, Evening Scale=0.8)
     B=&FloatFromCurve(FloatCurve=DataInterface() {propertyValues {Curve {keys [
          {interpMode=RCIM_Cubic, tangentMode=RCTM_User, time=0, value=0}
          {interpMode=RCIM_Cubic, tangentMode=RCTM_User, time=0.5, value=1}
          {tangentMode=RCTM_User, time=1, value=0}
         ]}}})
    )
    Scale RGB=&Multiply_VectorByFloat(Vector=<1 1 1>, Float=0.5)
    ScaleA=true
    ScaleRGB=true
   )
   &SubUVAnimation(
    End Frame Range Override=30
    "Sprite Renderer "=DataInterface() {propertyValues {SpriteRenderer=&self:FiresprintSubUV_0.NiagaraSpriteRendererProperties_0}}
    Start Frame Range Override=&RandomRangeInt(Minimum=0, Maximum=5)
    UseEndFrame=true
    UseStartFrame=true
   )
  ]
  rendererClasses [&NiagaraSpriteRendererProperties]
  renderers [{rendererClass=&NiagaraSpriteRendererProperties}]
 }
 Sparks {
  emitterUpdateScript [
   &InterpolateOverTime(
    Rate Of Change=3
    Target Float Value:NiagaraFloat=&MakeCustomFloatFromBool(
     Bool=&SetBoolByFloatComparison(A=&VectorLength(Vector=$Engine.Owner.Velocity), B=100)
     True Float=0
     False Float=1
    )
   )
   &EmitterState(Life Cycle Mode=::NewEnumerator1, Loop Duration=1, UseLoopDelay=false)
   &SpawnRate(
    SpawnRate=&Multiply_Float(
     A=30
     B=&FloatFromCurve(
      FloatCurve=DataInterface() {propertyValues {Curve {keys [<at 0.1 0>,<at 0.154 1>,<at 0.9421 1>,<at 1 0>]}}}
      CurveIndex=$System.Timeline.Time
     )
    )
    Use Spawn Probability=true
   )
  ]
  particleSpawnScript [
   &InitializeParticle(
    Color=<rgba 0 1.9086 3 1>
    Lifetime Max=1
    Lifetime Mode=::NewEnumerator1
    Sprite Size Max=<0.5 3>
    Sprite Size Min=<0.3 2>
    Sprite Size Mode=::NewEnumerator2
    UsePositionOffset=false
    Write Material Random=true
   )
   &CameraOffset(Camera Offset Amount=5)
   &SocketLocation(SamplePosOffset=false, Socket Index=2, Socket Reader=$System.Socket Reader, Socket Sampling Mode=::NewEnumerator3, VelocitySpeedLimit=false)
   &ShapeLocation(
    Box Midpoint=<-0.1 0 0.5>
    Box Size=<15 5 30>
    MaskSpawnGroup=false
    Override Seed=false
    Rotation Mode=::NewEnumerator1
    Shape Primitive=::NewEnumerator2
    Surface Only=true
    Surface Only Band Thickness=1
    UseOwnerScale=true
   )
  ]
  particleUpdateScript [
   &ParticleState()
   /SetVariables(Particles.Glow03_Color=&MakeLinearColorFromVectorAndFloat(
     "Vector (RGB)"=&MakeVectorFromLinearColorRGB(Color=&MultiplyLinearColor(Linear Color=<rgba 0 0.2042 1 0.2>, Scale Factor=&MakeVector4_FromFloat(Float=1)))
     "Float (Alpha)"=&Multiply_Float(A=0.2, B=1)
    )) {isSetParameters=true}
   /SetVariables(Particles.Glow03_Size=&Multiply_Vector2D_ByFloat(
     Float=&FloatFromCurve(
      FloatCurve=DataInterface() {propertyValues {Curve {keys [
          <at 0 0 RCIM_Cubic>
          <at 0.2 1.3 RCIM_Cubic>
          {arriveTangent=-0.1423, interpMode=RCIM_Cubic, leaveTangent=-0.1423, tangentMode=RCTM_User, time=1, value=0}
         ]}}}
      Scale Curve=10
     )
     Vector2D=<0.2 0.7>
    )) {isSetParameters=true}
   &ScaleColor(
    Scale Alpha=&FloatFromCurve(FloatCurve=DataInterface() {propertyValues {Curve {keys [<at 0.3 1>,<at 0.4 1>,<at 1 1>]}}})
    ScaleA=true
    ScaleRGB=true
   )
   &ScaleSpriteSize(Uniform Curve Sprite Scale=DataInterface() {propertyValues {Curve {keys [<at 0 0 RCIM_Cubic>,<at 0.2804 1 RCIM_Cubic>,<at 1 0 RCIM_Cubic>]}}}) {module=ScaleSpriteSize001}
   &CurlNoiseForce(
    Bias Noise=false
    Noise Frequency=300
    Noise Strength=&Multiply_Float(A=130, B=1)
    Pan Noise=true
    Pan Noise Field=<0 0 1>
    Random Seed=10
   )
   &Drag(Drag=4, Use Linear Drag=true, Use Rotational Drag=false)
   &GravityForce(Gravity=<0 0 100>)
   &SolveForcesAndVelocity(Clamp Velocity=false, Limit Acceleration=false)
  ]
  rendererClasses [&NiagaraSpriteRendererProperties]
  renderers [
   {rendererClass=&NiagaraSpriteRendererProperties}
   {rendererClass=&NiagaraSpriteRendererProperties, rendererIndex=1}
  ]
 }
}
summary {
 systemName=NS_Backpack_BaskIsle
 emitters [
  {emitterName=Sparks, rendererClasses [&NiagaraSpriteRendererProperties]}
  {emitterName=FireSpline, rendererClasses [&NiagaraMeshRendererProperties]}
  {emitterName=FireHead, rendererClasses [&NiagaraMeshRendererProperties]}
  {emitterName=FiresprintSubUV, rendererClasses [&NiagaraSpriteRendererProperties]}
  {emitterName=EyeGlow, rendererClasses [&NiagaraSpriteRendererProperties]}
 ]
 userVariables [
  User.bIsFrontEnd {type=NiagaraBool}
  User.bIsFrontEndPreview {type=NiagaraBool}
  User.DebugReactivity {type=NiagaraBool}
  User.ReactivityChestCount {type=NiagaraInt32}
 ]
}
