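# Material function for the BaskIsle backpack cosmetic that adds an animated ghostly fire effect to incoming material attributes.
# Panned local-space noise builds fire, fresnel and wipe masks alongside procedural face, arms and pages masks.
# It overrides base color, emissive, dithered gradient opacity, optional WPO wobble and normals perturbed by screen-space derivatives of that wobble.
# A Reactivity value comes from timestamped in/hold/out phases, or a looping debug switch.
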MaterialFunction /BRCosmetics/Accessories/FORT_Backpacks/Backpack_F_MED_BaskIsle/Materials/MF_BaskIsle_Backpack
graph {nodes [
  FunctionOutput(None=NamedRerouteUsage())
  n1 = Power(Base=NamedRerouteUsage(), Exp=ScalarParameter() {defaultValue=1, group=BackIsle, parameterName=Exp_1})
  n2 = ScalarParameter() {group=BackIsle, parameterName="Contrast (S)_1"}
  n3 = NamedRerouteUsage()
  n4 = Multiply(
   A=OneMinus(None=NamedRerouteUsage())
   B=Multiply(
    A=MaterialFunctionCall(In=@n1, Contrast=@n2) {materialFunction=&CheapContrast}.Result
    B=OneMinus(None=@n3)
   )
  )
  n5 = NamedRerouteUsage()
  NamedRerouteDeclaration(None=LinearInterpolate(A=NamedRerouteUsage(), Alpha=Multiply(A=@n4, B=@n5)) {constB=0}) {name=OutBaseColor, nodeColor=<rgba 0 1 0.3203 1>}
  n6 = VectorParameter() {group=BackIsle, parameterName=EffectColor, defaultValue=<rgba 0 0.1055 0.3229 1>}
  n7 = OneMinus(None=NamedRerouteUsage())
  n8 = ScalarParameter() {defaultValue=5, group=BackIsle, parameterName=InverseAOMultiplier}
  n9 = Add(
   A=Add(A=NamedRerouteUsage(), B=Multiply(A=@n7, B=@n8))
   B=Multiply(A=NamedRerouteUsage(), B=ScalarParameter() {defaultValue=2, group=BackIsle, parameterName=B_1}) {constB=3}
  )
  n10 = OneMinus(None=NamedRerouteUsage())
  n11 = NamedRerouteUsage()
  NamedRerouteDeclaration(None=LinearInterpolate(A=NamedRerouteUsage(), B=Multiply(A=@n6.RGB, B=@n9), Alpha=Multiply(A=@n10, B=@n11))) {name=OutEmissive, nodeColor=<rgba 1 0 0 1>}
  n12 = LocalPosition() {localOrigin=InstancePreSkinning}
  n13 = Multiply(
   A=ShaderStageSwitch(PixelShader=VertexInterpolator(VS=@n12.XYZ).PS, VertexShader=@n12.XYZ)
   B=Multiply(A=Constant() {r=0.01}, B=ScalarParameter() {defaultValue=7, group=BackIsle, parameterName=FireTextureScale})
  ) {constB=0.01}
  n14 = VectorParameter() {group=BackIsle, parameterName=FireSpeed, defaultValue=<rgba 0.05 -1.0322 0 1>}
  n15 = TextureSample(UVs=Panner(Coordinate=ComponentMask(None=@n13) {b=true, r=true}, Speed=ComponentMask(None=@n14.RGB) {g=true, r=true}) {speedX=0.25, speedY=0.05}) {
   samplerType=SAMPLERTYPE_LinearColor
   texture=/BRCosmetics/Characters/Player/Male/Medium/Bodies/M_MED_EggnogFaucet/Materials/T_EggnogFaucet_mask
  }
  n16 = OneMinus(None=@n15.G)
  n17 = LocalPosition() {localOrigin=InstancePreSkinning}
  n18 = Multiply(A=Subtract(A=Divide(A=@n17.Z) {constB=25}) {constB=-0.288}) {constB=0.15}
  n19 = ScalarParameter() {defaultValue=0.312, group=BackIsle, parameterName=WipeGradientOffset}
  n20 = Subtract(A=VertexInterpolator(VS=Subtract(A=@n18, B=@n19)).PS, B=NamedRerouteUsage())
  n21 = SmoothStep(Max=ScalarParameter() {defaultValue=1, parameterName=Max}, Value=Saturate(None=Sine(None=@n20))) {constMax=0.9}
  NamedRerouteDeclaration(None=Saturate(None=SmoothStep(Min=@n16, Value=@n21))) {name=FXMask, nodeColor=<rgba 1 0 0.8672 1>}
  n22 = SmoothStep(
   Min=@n16
   Value=Fresnel(ExponentIn=ScalarParameter() {defaultValue=0.5, group=BackIsle, parameterName=FireFresnelExponent}) {baseReflectFraction=0, exponent=1}
  )
  n23 = ComponentMask(None=PreSkinnedNormal()) {b=true}
  n24 = Power(Base=OneMinus(None=Abs(None=@n23))) {constExponent=5}
  NamedRerouteDeclaration(None=Multiply(A=Saturate(None=@n22), B=VertexInterpolator(VS=@n24).PS)) {name=FresnelFire, nodeColor=<rgba 0.6406 0 1 1>}
  n25 = GetMaterialAttributes(None=NamedRerouteUsage()) {attributeGetTypes [
    #1
    #2
    #3
    #4
    #5
   ]}
  NamedRerouteDeclaration(None=@n25.OpacityMask) {name=InOpacityMask, nodeColor=<rgba 1 0.4453 0 1>}
  n26 = ScalarParameter() {group=BackIsle, parameterName=GhostFadingMin}
  n27 = ScalarParameter() {defaultValue=0.05, group=BackIsle, parameterName=GhostFadingMax}
  n28 = Subtract(
   A=Add(A=LocalPosition().XYZ, B=VectorParameter() {group=BackIsle, parameterName=LocalPositionOffset, defaultValue=<rgba 0 15.2215 0 1>}.RGB)
   B=VectorParameter() {group=BackIsle, parameterName=BookPosition, defaultValue=<rgba 0 2.3994 0 1>}.RGB
  )
  n29 = ComponentMask(None=Multiply(A=@n28) {constB=0.1}) {g=true}
  n30 = Saturate(None=ConstantBiasScale(None=Multiply(A=@n29) {constB=-1}))
  n31 = Add(A=ComponentMask(None=@n28) {b=true}, B=ScalarParameter() {defaultValue=62.5235, group=BackIsle, parameterName=VerticalGradientOffset}) {constB=80}
  n32 = Saturate(None=Multiply(A=Multiply(A=@n31) {constB=0.01}, B=ScalarParameter() {defaultValue=1.5, group=BackIsle, parameterName=VerticalGradientScale}))
  n33 = LinearInterpolate(
   A=NamedRerouteUsage()
   B=Max(A=SmoothStep(Min=@n26, Max=@n27, Value=@n30) {constMax=0.1, constMin=0.8}, B=SmoothStep(Value=@n32) {constMin=0.9})
   Alpha=NamedRerouteUsage()
  )
  NamedRerouteDeclaration(None=MaterialFunctionCall(Alpha Threshold=VertexInterpolator(VS=@n33).PS) {materialFunction=&DitherTemporalAA_Fortnite}.Result) {name=OutOpacityMask, nodeColor=<rgba 0.125 0 1 1>}
  NamedRerouteDeclaration(None=@n25.EmissiveColor) {name=InEmissive, nodeColor=<rgba 0 1 0.2734 1>}
  NamedRerouteDeclaration(None=@n25.WorldPositionOffset) {name=InWPO, nodeColor=<rgba 0 0.5312 1 1>}
  n34 = Frac(None=Multiply(A=Time()) {constB=0.4})
  n35 = Time()
  n36 = ScalarParameter() {defaultValue=-999999, group=BackIsle, parameterName=Timestamp}
  n37 = Subtract(A=@n35, B=@n36)
  n38 = ScalarParameter() {defaultValue=1, group=BackIsle, parameterName=TimeStampInLength}
  n39 = ScalarParameter() {defaultValue=3.7, group=BackIsle, parameterName=TimeStampOnLength}
  n40 = Subtract(A=@n35, B=Add(A=@n36, B=Add(A=@n38, B=@n39) {constB=3, desc=eXTRAdURATION}))
  n41 = ScalarParameter() {defaultValue=2.5, group=BackIsle, parameterName=TimeStampOutLength}
  n42 = Add(A=Saturate(None=Divide(A=@n37, B=@n38) {desc=InLength}), B=Saturate(None=Divide(A=@n40, B=@n41) {constB=3, desc=OutLength}))
  NamedRerouteDeclaration(None=StaticSwitchParameter(True=Saturate(None=@n34), False=Divide(A=@n42)) {parameterName=DebugAnimation}) {name=Reactivity, nodeColor=<rgba 1 0 0.9141 1>}
  NamedRerouteDeclaration(None=FunctionInput() {inputName=Material, inputType=FunctionInput_MaterialAttributes}) {name=InMaterial, nodeColor=<rgba 0 0.1328 1 1>}
  n43 = SetMaterialAttributes!(
   MaterialAttributes=NamedRerouteUsage()
   OpacityMask=NamedRerouteUsage()
   WorldPositionOffset=NamedRerouteUsage()
   EmissiveColor=NamedRerouteUsage()
   BaseColor=NamedRerouteUsage()
   Normal=NamedRerouteUsage()
  ) {attributeSetTypes [
    #3
    #2
    #4
    #1
    #5
   ]}
  NamedRerouteDeclaration(None=@n43) {name=OutMaterial, nodeColor=<rgba 0.4531 1 0 1>}
  n44 = NamedRerouteUsage()
  n45 = Constant() {r=1}
  NamedRerouteDeclaration(None=OneMinus(None=Desaturation(None=@n44, Fraction=@n45))) {name=BookFXMask, nodeColor=<rgba 1 0 0.1172 1>}
  n46 = FunctionInput(Preview=Constant() {r=1}) {bUsePreviewValueAsDefault=true, description=FXMaskChannel, inputName=TechArtMask, sortPriority=1, previewValue=<1 1 1 1>}
  NamedRerouteDeclaration(None=ChannelMaskParameter(None=@n46) {group=BackIsle, parameterName=BaskIsleGhostMask}.R) {name=InGhostMask, nodeColor=<rgba 1 0 0.7734 1>}
  n47 = ScalarParameter() {defaultValue=0.49, group=BackIsle}
  n48 = Constant2Vector() {g=0.6, r=1}
  NamedRerouteDeclaration(None=MaterialFunctionCall!(A=TextureCoordinate(), B=Constant2Vector() {g=0.144, r=0.75}, Bounds=Multiply(A=@n47, B=@n48), Edge Falloff=Constant()) {materialFunction=/Engine/Functions/Engine_MaterialFunctions01/Gradient/BoxMask-2D}.Result) {name=PagesMask, nodeColor=<rgba 1 0.0703 0 1>}
  n49 = LocalPosition() {localOrigin=InstancePreSkinning}
  n50 = ShaderStageSwitch(PixelShader=VertexInterpolator(VS=@n49.XYZ).PS, VertexShader=@n49.XYZ)
  n51 = Multiply(A=@n50, B=Multiply(A=Constant() {r=0.01}, B=ScalarParameter() {defaultValue=5, group=BackIsle, parameterName=WPOTextureScale})) {constB=0.01}
  n52 = Constant() {r=0.22}
  n53 = VectorParameter() {group=BackIsle, parameterName=WPOSpeed, defaultValue=<rgba 0.25 0.05 0 1>}
  n54 = TextureSample(UVs=Panner(
    Coordinate=MaterialFunctionCall(UVs=@n51, "Rotation Angle (0-1)"=@n52) {materialFunction=&CustomRotator}.Rotated Values
    Speed=ComponentMask(None=@n53.RGB) {g=true, r=true}
   ) {speedX=0.25, speedY=0.05}) {
   samplerType=SAMPLERTYPE_LinearColor
   texture=&T_EggnogFaucet_mask
  }
  n55 = ConstantBiasScale(None=@n54.G) {bias=-0.5, scale=2}
  n56 = ConstantBiasScale(None=ComponentMask(None=PreSkinnedNormal()) {b=true})
  n57 = VertexInterpolator(VS=@n56)
  n58 = LinearInterpolate(
   A=ScalarParameter() {defaultValue=0.8, group=BackIsle, parameterName=WPOIntensity_Head}
   B=Multiply(
    A=ScalarParameter() {defaultValue=0.3, group=BackIsle, parameterName=WPOIntensity_Arms}
    B=ShaderStageSwitch(PixelShader=@n57.PS, VertexShader=@n56)
   )
   Alpha=NamedRerouteUsage()
  )
  n59 = Multiply(A=Multiply(A=Multiply(A=@n55, B=@n58), B=NamedRerouteUsage()), B=NamedRerouteUsage())
  n60 = NamedRerouteUsage()
  NamedRerouteDeclaration(None=StaticSwitchParameter(True=Add(A=@n59, B=@n60), False=@n60) {defaultValue=true, group=BackIsle, parameterName=UseWPO}) {name=OutWPO, nodeColor=<rgba 0 1 0.9062 1>}
  n61 = FunctionInput() {
   bUsePreviewValueAsDefault=true
   description=FXMaskChannel
   inputName=InMask
   inputType=FunctionInput_Scalar
   sortPriority=1
   previewValue=<1 1 1 1>
  }
  n62 = MaterialFunctionCall(Color=@n61) {materialFunction=&SplitComponents}
  NamedRerouteDeclaration(None=@n62.G) {name=InGMap, nodeColor=<rgba 1 0.2344 0 1>}
  n63 = ComponentMask(None=Multiply(A=@n50) {constB=0.01}) {r=true}
  NamedRerouteDeclaration(None=SmoothStep(Value=Abs(None=@n63)) {constMax=0.1, constMin=0.07}) {name=ArmsMask, nodeColor=<rgba 0 1 0.0391 1>}
  NamedRerouteDeclaration(None=@n25.BaseColor) {name=InBaseColor, nodeColor=<rgba 0.9453 0 1 1>}
  NamedRerouteDeclaration(None=@n62.R) {name=InAO, nodeColor=<rgba 1 0.3984 0 1>}
  n64 = Multiply(A=AppendVector(A=DDX(Value=@n59), B=DDY(Value=@n59))) {constB=-1}
  NamedRerouteDeclaration(None=MaterialFunctionCall(BaseNormal=NamedRerouteUsage(), AdditionalNormal=DeriveNormalZ(InXY=@n64)) {materialFunction=&BlendAngleCorrectedNormals}.Result) {name=OutNormal, nodeColor=<rgba 0.3359 1 0 1>}
  NamedRerouteDeclaration(None=@n25.Normal) {name=InNormal, nodeColor=<rgba 0.0312 1 0 1>}
  n65 = Multiply(
   A=Add(A=@n50, B=VectorParameter() {group=BackIsle, parameterName=FaceMaskOffset, defaultValue=<rgba 0 11.2965 -8.285 1>}.RGB)
   B=VectorParameter() {group=BackIsle, parameterName=FaceMaskScaleRatio, defaultValue=<rgba 1.5 1 0.75 1>}.RGB
  )
  n66 = Constant()
  n67 = ScalarParameter() {defaultValue=8.9, group=BackIsle, parameterName=FaceMaskRadius}
  NamedRerouteDeclaration(None=OneMinus(None=SphereMask(A=@n65, B=@n66, Radius=@n67) {attenuationRadius=10, hardnessPercent=30})) {name=FaceMask, nodeColor=<rgba 0.1953 0 1 1>}
 ]}
