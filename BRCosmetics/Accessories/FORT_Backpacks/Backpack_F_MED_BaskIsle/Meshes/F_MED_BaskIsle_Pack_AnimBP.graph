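# Anim Blueprint for the BaskIsle backpack cosmetic that plays an open-scare animation when its wearer opens a chest, or on timers in the front end.
# Separate in-game and front-end state machines are blended by bIsInFrontEnd.
# The pack bone then gets spine-driven offsets and sine-wave bobbing, followed by curve-weighted AnimDynamics on the right strap.
# ChestOpened input events also trigger throttled material timestamp and VFX chest-count updates.
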AnimBlueprint /BRCosmetics/Accessories/FORT_Backpacks/Backpack_F_MED_BaskIsle/Meshes/F_MED_BaskIsle_Pack_AnimBP
target_skeleton=&/F_MED_BaskIsle_Pack_Skeleton
graphs [
 AnimGraph {nodes [
   AnimGraphNode_StateMachine_0 = BaskIsle_StateMachine() {is=title AnimGraphNode_StateMachine}
   AnimGraphNode_StateMachine_1 = FE_Reactivity() {is=title AnimGraphNode_StateMachine}
   n1 = Get bIsInFrontEnd()
   n2 = "Transform (Modify) Bone - Bone: pack"(
    ComponentPose=Local To Component(LocalPose=Blend Poses by bool(BlendPose_0=@AnimGraphNode_StateMachine_1, BlendPose_1=@AnimGraphNode_StateMachine_0, bActiveValue=@n1) {blendType=Linear})
    Translation="0,-0.5,-2"
    Rotation="0,0,7.5"
   ) {
    rotationMode=BMM_Additive
    translationMode=BMM_Additive
    translationSpace=BCS_BoneSpace
    alphaBoolBlend {blendOption=Linear}
    boneToModify {boneName=pack}
    rotation {roll=7.5}
    translation {y=-0.5, z=-2}
   }
   n3 = "Transform (Modify) Bone - Bone: pack"(ComponentPose=@n2, Translation=Get backpack_translation(), Rotation=Get backpack_rotate()) {rotationMode=BMM_Additive, translationMode=BMM_Additive, alphaBoolBlend {blendOption=Linear}, boneToModify {boneName=pack}}
   n4 = "Transform (Modify) Bone - Bone: pack"(ComponentPose=@n3, Alpha=0.5, Translation=Get Ring_A_Float() -> (Ring_A_Float:Vector), Rotation="0,0,0") {translationMode=BMM_Additive, translationSpace=BCS_BoneSpace, alphaBoolBlend {blendOption=Linear}, boneToModify {boneName=pack}}
   n5 = Anim Dynamics - Bone: strap_r_02_jnt!(ComponentPose=@n4, AlphaCurveName:Name=DynamicsAlpha, ExternalForce="0,0,0") -> (Pose) {
    alphaInputType=Curve
    angularDampingOverride=1
    angularSpringConstant=10000
    bAngularSpring=true
    bOverrideAngularDamping=true
    bUseGravityOverride=true
    gravityScale=0.35
    simulationSpace=RootRelative
    alphaBoolBlend {blendOption=Linear}
    boundBone {boneName=strap_r_02_jnt}
    componentAppliedLinearAccClamp=<3000 3000 3000>
    componentLinearAccScale=<1 1 1>
    componentLinearVelScale=<1 1 1>
    gravityOverride {z=-980}
    physicsBodyDefinitions [{
      boundBone {boneName=strap_r_02_jnt}
      boxExtents=<40 40 40>
      constraintSetup {angularLimitsMax=<0 12 1>, angularLimitsMin=<0 -7.5 -1>, angularTarget=<1 0 0>}
      localJointOffset=<5.5 0 0>
     }]
   }
   n6 = Anim Dynamics - Bone: strap_r_03_jnt!(ComponentPose=@n5, AlphaCurveName:Name=DynamicsAlpha, ExternalForce="0,0,0") -> (Pose) {
    alphaInputType=Curve
    angularDampingOverride=1
    angularSpringConstant=10000
    bAngularSpring=true
    bOverrideAngularDamping=true
    bUseGravityOverride=true
    gravityScale=0.35
    simulationSpace=RootRelative
    alphaBoolBlend {blendOption=Linear}
    boundBone {boneName=strap_r_03_jnt}
    componentAppliedLinearAccClamp=<3000 3000 3000>
    componentLinearAccScale=<1 1 1>
    componentLinearVelScale=<1 1 1>
    gravityOverride {z=-980}
    physicsBodyDefinitions [{
      boundBone {boneName=strap_r_03_jnt}
      boxExtents=<40 40 40>
      constraintSetup {angularLimitsMax=<1 20 1>, angularLimitsMin=<-1 -7.5 -1>, angularTarget=<1 0 0>}
      localJointOffset=<5.5 0 0>
     }]
   }
   n7 = Anim Dynamics - Bone: strap_r_04_jnt!(ComponentPose=@n6, AlphaCurveName:Name=DynamicsAlpha, ExternalForce="0,0,0") -> (Pose) {
    alphaInputType=Curve
    angularDampingOverride=1
    angularSpringConstant=10000
    bAngularSpring=true
    bOverrideAngularDamping=true
    bUseGravityOverride=true
    gravityScale=0.35
    simulationSpace=RootRelative
    alphaBoolBlend {blendOption=Linear}
    boundBone {boneName=strap_r_04_jnt}
    componentAppliedLinearAccClamp=<3000 3000 3000>
    componentLinearAccScale=<1 1 1>
    componentLinearVelScale=<1 1 1>
    gravityOverride {z=-980}
    physicsBodyDefinitions [{
      boundBone {boneName=strap_r_04_jnt}
      boxExtents=<40 40 40>
      constraintSetup {angularLimitsMax=<1 20 1>, angularLimitsMin=<-1 -17 -1>, angularTarget=<1 0 0>}
      localJointOffset=<5.5 0 0>
     }]
   }
   n8 = Anim Dynamics - Bone: strap_r_05_jnt!(ComponentPose=@n7, AlphaCurveName:Name=DynamicsAlpha, ExternalForce="0,0,0") -> (Pose) {
    alphaInputType=Curve
    angularDampingOverride=1
    angularSpringConstant=10000
    bAngularSpring=true
    bOverrideAngularDamping=true
    bUseGravityOverride=true
    gravityScale=0.35
    simulationSpace=RootRelative
    alphaBoolBlend {blendOption=Linear}
    boundBone {boneName=strap_r_05_jnt}
    componentAppliedLinearAccClamp=<3000 3000 3000>
    componentLinearAccScale=<1 1 1>
    componentLinearVelScale=<1 1 1>
    gravityOverride {z=-980}
    physicsBodyDefinitions [{
      boundBone {boneName=strap_r_05_jnt}
      boxExtents=<40 40 40>
      constraintSetup {angularLimitsMax=<1 20 1>, angularLimitsMin=<-1 -20 -1>, angularTarget=<1 0 0>}
      localJointOffset=<5.5 0 0>
     }]
   }
   n9 = Anim Dynamics - Bone: strap_r_06_jnt!(ComponentPose=@n8, AlphaCurveName:Name=DynamicsAlpha, ExternalForce="0,0,0") -> (Pose) {
    alphaInputType=Curve
    angularDampingOverride=1
    angularSpringConstant=10000
    bAngularSpring=true
    bOverrideAngularDamping=true
    bUseGravityOverride=true
    gravityScale=0.35
    simulationSpace=RootRelative
    alphaBoolBlend {blendOption=Linear}
    boundBone {boneName=strap_r_06_jnt}
    componentAppliedLinearAccClamp=<3000 3000 3000>
    componentLinearAccScale=<1 1 1>
    componentLinearVelScale=<1 1 1>
    gravityOverride {z=-980}
    physicsBodyDefinitions [{
      boundBone {boneName=strap_r_06_jnt}
      boxExtents=<40 40 40>
      constraintSetup {angularLimitsMax=<1 20 1>, angularLimitsMin=<-1 -20 -1>, angularTarget=<1 0 0>}
      localJointOffset=<5.5 0 0>
     }]
   }
   Output Pose(Result=Component To Local(ComponentPose=@n9)) {layerGroup=None}
  ]}
 BaskIsle_StateMachine {
  entry=Idle
  path=AnimGraph.AnimGraphNode_StateMachine_0.BaskIsle_StateMachine
  ids=id
  nodes [
   AnimStateNode_0 = Idle() {
    stateEntered {guid=#1}
    stateFullyBlended {guid=#2}
    stateLeft {guid=#3}
   }
   AnimStateNode_4 = ChestReaction() {is=title AnimStateNode}
   AnimStateTransitionNode_1 = AnimStateTransitionNode() {bAutomaticRuleBasedOnSequencePlayerInState=true, crossfadeDuration=0}
   AnimStateTransitionNode_12 = AnimStateTransitionNode() {crossfadeDuration=0}
  ]
  links [
   {
    dst=Idle
    rule=AnimGraph.AnimGraphNode_StateMachine_0.BaskIsle_StateMachine.AnimStateTransitionNode_1.Transition
    src=ChestReaction
    via=AnimStateTransitionNode_1
   }
   {
    dst=ChestReaction
    rule=AnimGraph.AnimGraphNode_StateMachine_0.BaskIsle_StateMachine.AnimStateTransitionNode_12.Transition
    src=Idle
    via=AnimStateTransitionNode_12
   }
  ]
 }
 Idle {
  path=AnimGraph.AnimGraphNode_StateMachine_0.BaskIsle_StateMachine.AnimStateNode_0.Idle
  nodes [
   n1 = "Transform (Modify) Bone - Bone: neck_base"!(
    ComponentPose=Local To Component(LocalPose=Sequence Player 'BaskIsle_Idle_PACK'() {sequence=/BRCosmetics/Accessories/FORT_Backpacks/Backpack_F_MED_BaskIsle/Animation/BaskIsle_Idle_PACK})
    bAlphaBoolEnabled:bool=True
    Translation="0,0,0"
    Rotation="0,0,0"
    Scale="0,0,0"
   ) -> (Pose) {
    alphaInputType=Bool
    scaleMode=BMM_Additive
    scaleSpace=BCS_BoneSpace
    translationMode=BMM_Additive
    translationSpace=BCS_BoneSpace
    alphaBoolBlend {blendOption=Linear}
    boneToModify {boneName=neck_base}
    scale=<0 0 0>
   }
   n2 = "Transform (Modify) Bone - Bone: lowerarm_r_jnt"!(ComponentPose=@n1, bAlphaBoolEnabled:bool=True, Translation="0,0,0", Rotation="0,0,0", Scale="0,0,0") -> (Pose) {
    alphaInputType=Bool
    scaleMode=BMM_Additive
    scaleSpace=BCS_BoneSpace
    translationMode=BMM_Additive
    translationSpace=BCS_BoneSpace
    alphaBoolBlend {blendOption=Linear}
    boneToModify {boneName=lowerarm_r_jnt}
    scale=<0 0 0>
   }
   n3 = "Transform (Modify) Bone - Bone: lowerarm_l_jnt"!(ComponentPose=@n2, bAlphaBoolEnabled:bool=True, Translation="0,0,0", Rotation="0,0,0", Scale="0,0,0") -> (Pose) {
    alphaInputType=Bool
    scaleMode=BMM_Additive
    scaleSpace=BCS_BoneSpace
    translationMode=BMM_Additive
    translationSpace=BCS_BoneSpace
    alphaBoolBlend {blendOption=Linear}
    boneToModify {boneName=lowerarm_l_jnt}
    scale=<0 0 0>
   }
   Output Animation Pose(Result=Component To Local(ComponentPose=@n3))
  ]
 }
 ChestReaction {
  path=AnimGraph.AnimGraphNode_StateMachine_0.BaskIsle_StateMachine.AnimStateNode_4.ChestReaction
  nodes [Output Animation Pose(Result=Sequence Player 'BaskIsle_OpenScare_PACK'() {bLoopAnimation=false, sequence=/BRCosmetics/Accessories/FORT_Backpacks/Backpack_F_MED_BaskIsle/Animation/BaskIsle_OpenScare_PACK})]
 }
 Transition {path=AnimGraph.AnimGraphNode_StateMachine_0.BaskIsle_StateMachine.AnimStateTransitionNode_1.Transition, nodes [Result()]}
 Transition {
  path=AnimGraph.AnimGraphNode_StateMachine_0.BaskIsle_StateMachine.AnimStateTransitionNode_12.Transition
  nodes [Result(bCanEnterTransition=Get bChestOpened() -> (bChestOpened:bool))]
 }
 FE_Reactivity {
  entry=Idle
  path=AnimGraph.AnimGraphNode_StateMachine_1.FE_Reactivity
  ids=id
  nodes [
   AnimStateNode_0 = Idle()
   AnimStateNode_2 = FE_ChestOpen() {is=title AnimStateNode, stateEntered {guid=#4, notifyName=Entered FE_Reactive_Anim}}
   AnimStateTransitionNode_8 = AnimStateTransitionNode()
   AnimStateTransitionNode_9 = AnimStateTransitionNode() {bAutomaticRuleBasedOnSequencePlayerInState=true}
  ]
  links [
   {
    dst=Idle
    rule=AnimGraph.AnimGraphNode_StateMachine_1.FE_Reactivity.AnimStateTransitionNode_9.Transition
    src=FE_ChestOpen
    via=AnimStateTransitionNode_9
   }
   {
    dst=FE_ChestOpen
    rule=AnimGraph.AnimGraphNode_StateMachine_1.FE_Reactivity.AnimStateTransitionNode_8.Transition
    src=Idle
    via=AnimStateTransitionNode_8
   }
  ]
 }
 Idle {
  path=AnimGraph.AnimGraphNode_StateMachine_1.FE_Reactivity.AnimStateNode_0.Idle
  nodes [
   n1 = "Transform (Modify) Bone - Bone: neck_base"(
    ComponentPose=Local To Component(LocalPose=Sequence Player 'BaskIsle_Idle_PACK'() {bLoopAnimation=false, sequence=&BaskIsle_Idle_PACK})
    bAlphaBoolEnabled=True
    Translation="0,0,0"
    Rotation="0,0,0"
    Scale="0,0,0"
   ) {
    alphaInputType=Bool
    scaleMode=BMM_Additive
    scaleSpace=BCS_BoneSpace
    translationMode=BMM_Additive
    translationSpace=BCS_BoneSpace
    alphaBoolBlend {blendOption=Linear}
    boneToModify {boneName=neck_base}
    scale=<0 0 0>
   }
   n2 = "Transform (Modify) Bone - Bone: lowerarm_r_jnt"(ComponentPose=@n1, bAlphaBoolEnabled=True, Translation="0,0,0", Rotation="0,0,0", Scale="0,0,0") {
    alphaInputType=Bool
    scaleMode=BMM_Additive
    scaleSpace=BCS_BoneSpace
    translationMode=BMM_Additive
    translationSpace=BCS_BoneSpace
    alphaBoolBlend {blendOption=Linear}
    boneToModify {boneName=lowerarm_r_jnt}
    scale=<0 0 0>
   }
   n3 = "Transform (Modify) Bone - Bone: lowerarm_l_jnt"(ComponentPose=@n2, bAlphaBoolEnabled=True, Translation="0,0,0", Rotation="0,0,0", Scale="0,0,0") {
    alphaInputType=Bool
    scaleMode=BMM_Additive
    scaleSpace=BCS_BoneSpace
    translationMode=BMM_Additive
    translationSpace=BCS_BoneSpace
    alphaBoolBlend {blendOption=Linear}
    boneToModify {boneName=lowerarm_l_jnt}
    scale=<0 0 0>
   }
   Output Animation Pose(Result=Component To Local(ComponentPose=@n3))
  ]
 }
 FE_ChestOpen {
  path=AnimGraph.AnimGraphNode_StateMachine_1.FE_Reactivity.AnimStateNode_2.FE_ChestOpen
  nodes [Output Animation Pose(Result=Sequence Player 'BaskIsle_OpenScare_PACK'() {bLoopAnimation=false, sequence=&BaskIsle_OpenScare_PACK})]
 }
 Transition {
  path=AnimGraph.AnimGraphNode_StateMachine_1.FE_Reactivity.AnimStateTransitionNode_8.Transition
  nodes [Result(bCanEnterTransition=Get Play Reactive FE Anim() -> (Play Reactive FE Anim:bool))]
 }
 Transition {path=AnimGraph.AnimGraphNode_StateMachine_1.FE_Reactivity.AnimStateTransitionNode_9.Transition, nodes [Result()]}
 EventGraph {nodes [
   Reactivity() {is=title K2Node_CustomEvent}
   > n1 = Get!(execute:Exec) -> (then:Exec, else:Exec, BackpackPartModifier:Custom Character Part Modifier Object Reference)
   > Do Once(Reset=@n3.then)
   > n2 = Sequence()
   > SetScalarParameterValueOnMaterials!(
    execute
    self:Mesh Component Object Reference=GetSkeletalMeshForPartOnPlayerPawn!(self:Custom Character Part Modifier Object Reference=@n1) -> (ReturnValue:Mesh Component Object Reference)
    ParameterName:Name=Timestamp
    ParameterValue:float=GetGameTimeInSeconds()
   ) -> (then)
   > n3 = Delay(Duration=6)
   n4 = Increment Int(" "=@n2.then_1, Value=Get ChestCountFX() -> (ChestCountFX:int=0))
   > SetIntParameter!(
    execute
    self:FXSystem Component Object Reference=GetIdleVFXComponent!(self:Custom Character Part Modifier Object Reference=@n1) -> (ReturnValue:FXSystem Component Object Reference)
    ParameterName:Name=ReactivityChestCount
    Param:int=@n4
   ) -> (then)
   Event BlueprintBeginPlay() {is=title K2Node_Event}
   > Parent: BlueprintBeginPlay() {is=title K2Node_CallParentFunction}
   > Delay(Duration=0.2)
   > n5 = Cast To FortPlayerPawn(Object=TryGetFortPawn())
   > Set BackpackPartModifier(BackpackPartModifier:Custom Character Part Modifier Object Reference=GetModifierForPartType!(self:Fort Player Pawn Object Reference=@n5, PartType:EFortCustomPartType Enum=Backpack) -> (ReturnValue:Custom Character Part Modifier Object Reference)) -> (Output_Get:Custom Character Part Modifier Object Reference)
   AnimNotify_FE_ReactiveAnim() {is=title K2Node_Event}
   > Set Play Reactive FE Anim(Play Reactive FE Anim:bool=false) -> (Output_Get:bool)
   > n6 = Delay(Duration=10)
   Set Play Reactive FE Anim(execute=[@n6.then, @n19.then], Play Reactive FE Anim=true)
   Event BlueprintUpdateAnimation()
   n7 = GetRealTimeSeconds()
   n8 = "Sin (Radians)"(A=float * float(A=float * float(A=1.1, B=1.1), B=@n7))
   n9 = "Sin (Radians)"(A=float * float(A=float * float(A=1, B=1.1), B=@n7))
   n10 = "Sin (Radians)"(A=float * float(A=float * float(A=1.2, B=1.8), B=@n7))
   > Set Ring_A_Float(Ring_A_Float:Vector=MakeVector(
     X=Remap Float(Input=@n10, OldRange=(X=-1,Y=1), NewRange=(X=-1,Y=1)) -> (Output=~)
     Y=float * float(A=@n8, B=1.5)
     Z=float * float(A=@n9, B=4.5)
    )) -> (Output_Get:Vector)
   n11 = GetOwnerSkeletalMeshComponent()
   n12 = BreakRotator(InRot=BreakTransform(InTransform=GetDeltaTransformFromRefPose(self=@n11, BoneName=spine_05, BaseName=spine_01)).Rotation)
   > Set backpack_rotate(backpack_rotate=MakeRotator(Roll=MapRangeClamped(Value=@n12.Yaw, InRangeA=0, InRangeB=-30, OutRangeB=30)))
   > n13 = Set backpack_translation(backpack_translation=MakeVector(Y=MapRangeClamped(Value=@n12.Yaw, InRangeA=0, InRangeB=-30, OutRangeA=0.5, OutRangeB=0)))
   n14 = Event BlueprintProcessAnimInputEvent()
   n15 = IsNamed(self=@n14.AnimInputEvent, InEventName=aie.ChestOpened)
   n16 = Branch(execute=@n13.then, Condition=@n15)
   > Set IsIdle(IsIdle:bool=false) -> (Output_Get:bool)
   Set IsIdle(execute=@n16.else, IsIdle=true)
   Event BlueprintPostEvaluateAnimation() {is=title K2Node_Event}
   > Set bChestOpened(bChestOpened:bool=false) -> (Output_Get:bool)
   > Set Play Reactive FE Anim(Play Reactive FE Anim=false)
   Event OnCharacterCustomizationLoaded()
   > n17 = Cast To FortPlayerPawn(Object=TryGetPawnOwner())
   > n18 = Branch()
   > Do Once(Reset=@n20.then)
   > n19 = Delay(Duration=3)
   n20 = Set Play Reactive FE Anim(execute=@n18.else, Play Reactive FE Anim=false)
   TryGetPawnOwner()
   GetModifierForPartType(self=@n17, PartType=Backpack)
   Cast To B_Athena_PartModifier_Backpack_BasketSuburb!(execute, Object:Wildcard) -> (then, CastFailed, AsB Athena Part Modifier Backpack Basket Suburb:B Athena Part Modifier Backpack Basket Suburb Object Reference) {is=title K2Node_DynamicCast}
   Get FE_ReactivityEnabled!(self:B Athena Part Modifier Backpack Twinkle Loop Object Reference) -> (FE_ReactivityEnabled:bool)
   Set bChestOpened(execute=@n21.then, bChestOpened=false)
   Reactivity(execute=@n22.then) {is=title K2Node_CallFunction}
   > Set bChestOpened(bChestOpened=true)
   > n21 = RetriggerableDelay(Duration=0.2)
   n22 = Branch(execute=@n14.then, Condition=@n15)
  ]}
]
variables [
 AlphaFE {type=number}
 Anim Input Event {type=FortAnimInputEvent, default=null}
 backpack_rotate {type=Rotator}
 backpack_translation {type=Vector}
 BackpackPartModifier {type=CustomCharacterPartModifier, default=null}
 bChestOpened {type=boolean}
 ChestCountFX {type=integer}
 FrontEnd {type=boolean}
 Gravity Override Params {
  type=FortGravityOverrideParameterSet
  default=(Skydive_GoingUp=(GravityOverride=()),Skydive_FloatingDown=(GravityOverride=()),Skydive_DivingDown=(GravityOverride=()),Skydive_Parachuting=(GravityOverride=()),OnGround_Standing=(GravityOverride=()),OnGround_Moving=(GravityOverride=()),OnGround_Moving_HighRotationRate=(GravityOverride=()),OnGround_Crouch_Moving=(GravityOverride=()),Emote_or_Melee=(GravityOverride=()),Swimming=(GravityOverride=()),Front_End=(GravityOverride=()),DBNO=(GravityOverride=()),DBNO_Carried=(GravityOverride=()),OnGround_Falling=(GravityOverride=()),Ziplining=(GravityOverride=()),Skydive_Surfing=(GravityOverride=()),Riding_Standing=(GravityOverride=()),Riding_Moving=(GravityOverride=()),Riding_Moving_Targeting=(GravityOverride=()),Tactical_Sprint=(GravityOverride=()),Rail_Grinding=(GravityOverride=()),)
 }
 GravityOverride {type=Vector}
 IsIdle {type=boolean}
 OnZipline {type=boolean}
 Play Reactive FE Anim {type=boolean}
 Reactivity Enabled {type=boolean}
 Ring_A_Float {type=Vector}
 SimlationSpace {type=AnimPhysSimSpaceType}
 ZiplineAlpha {type=number}
]
