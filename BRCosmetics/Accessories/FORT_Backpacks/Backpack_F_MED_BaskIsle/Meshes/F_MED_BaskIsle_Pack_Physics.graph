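# Physics asset for the BaskIsle cosmetic backpack, providing bodies and joints for simulating its right-hand strap.
# A large capsule on the book root anchors a chain of six small strap capsules, with the first joint free to swing and twist.
# Later links chain in sequence with narrow swing limits and zero twist limit.
# Extra angular-free constraints also tie the second through fifth links directly to the strap tip.
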PhysicsAsset /BRCosmetics/Accessories/FORT_Backpacks/Backpack_F_MED_BaskIsle/Meshes/F_MED_BaskIsle_Pack_Physics
bodies {
 root_book_jnt [{shapeType=Capsule, length=12, radius=8, center=<5 0 0>, rotation=<rot 0 0 90>}]
 strap_r_01_jnt [{shapeType=Capsule, length=0.3, radius=0.75, shapeName=strap_r_01_jnt_capsule, center=<1.5 0 0>, rotation=<rot 90 0 0>}]
 strap_r_02_jnt [{shapeType=Capsule, length=0.3, radius=0.75, shapeName=strap_r_02_jnt_capsule, center=<1.5 0 0>, rotation=<rot 90 0 0>}]
 strap_r_03_jnt [{shapeType=Capsule, length=0.3, radius=0.75, shapeName=strap_r_03_jnt_capsule, center=<1.5 0 0>, rotation=<rot 90 0 0>}]
 strap_r_04_jnt [{shapeType=Capsule, length=0.3, radius=0.75, shapeName=strap_r_04_jnt_capsule, center=<1.5 0 0>, rotation=<rot 90 0 0>}]
 strap_r_05_jnt [{shapeType=Capsule, length=0.3, radius=0.75, shapeName=strap_r_05_jnt_capsule, center=<1.5 0 0>, rotation=<rot 90 0 0>}]
 strap_r_06_jnt [{shapeType=Capsule, length=0.3, radius=0.75, shapeName=strap_r_06_jnt_capsule, center=<1.5 0 0>, rotation=<rot 90 0 0>}]
}
constraints [
 {
  bone1Name=strap_r_01_jnt
  bone2Name=root_book_jnt
  swing1LimitDegrees=45
  swing1Motion=Free
  swing2LimitDegrees=45
  swing2Motion=Free
  twistLimitDegrees=45
  twistMotion=Free
 }
 {bone1Name=strap_r_02_jnt, bone2Name=strap_r_01_jnt, swing1LimitDegrees=3, swing2LimitDegrees=5, twistLimitDegrees=0}
 {bone1Name=strap_r_03_jnt, bone2Name=strap_r_02_jnt, swing1LimitDegrees=3, swing2LimitDegrees=20, twistLimitDegrees=0}
 {bone1Name=strap_r_04_jnt, bone2Name=strap_r_03_jnt, swing1LimitDegrees=3, swing2LimitDegrees=20, twistLimitDegrees=0}
 {bone1Name=strap_r_05_jnt, bone2Name=strap_r_04_jnt, swing1LimitDegrees=3, swing2LimitDegrees=20, twistLimitDegrees=0}
 {bone1Name=strap_r_06_jnt, bone2Name=strap_r_05_jnt, swing1LimitDegrees=3, swing2LimitDegrees=20, twistLimitDegrees=0}
 {
  bone1Name=strap_r_02_jnt
  bone2Name=strap_r_06_jnt
  swing1LimitDegrees=45
  swing1Motion=Free
  swing2LimitDegrees=45
  swing2Motion=Free
  twistLimitDegrees=45
  twistMotion=Free
 }
 {
  bone1Name=strap_r_03_jnt
  bone2Name=strap_r_06_jnt
  swing1LimitDegrees=45
  swing1Motion=Free
  swing2LimitDegrees=45
  swing2Motion=Free
  twistLimitDegrees=45
  twistMotion=Free
 }
 {
  bone1Name=strap_r_04_jnt
  bone2Name=strap_r_06_jnt
  swing1LimitDegrees=45
  swing1Motion=Free
  swing2LimitDegrees=45
  swing2Motion=Free
  twistLimitDegrees=45
  twistMotion=Free
 }
 {
  bone1Name=strap_r_05_jnt
  bone2Name=strap_r_06_jnt
  swing1LimitDegrees=45
  swing1Motion=Free
  swing2LimitDegrees=45
  swing2Motion=Free
  twistLimitDegrees=45
  twistMotion=Free
 }
]
